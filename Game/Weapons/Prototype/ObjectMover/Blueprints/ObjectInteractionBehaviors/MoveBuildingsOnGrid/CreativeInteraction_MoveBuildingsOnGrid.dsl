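; Creative move tool interaction behavior for moving buildings on a grid, layering preview transition effects and a placement sound onto parent logic.
; Server start spawns a TransitionIn preview per selected actor.
; Interaction end, only where cosmetic events cannot run, sets those previews to TransitionOut and queues backwards playback after the current animation.
; Object placement spawns a TransitionOut preview for the first component's owner, while triggering plays the place sound at the owner's location.
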(event CreativeMoveTool|EventEndCreativeInteraction
  (bind _returnvalue (CreativeInteraction|GetFortCreativeMoveTool))
  ;[1 1 1] On place
  (|CallParentFunction)
  (Networking|CanExecuteCosmeticEvents
    (:False
      (for _array_element (CreativeMoveTool|GetOwnerSelectedActors)
        (Utilities|IsValid (Class|FortCreativeMoveTool|FindActivePreviewforObject _returnvalue _array_element)
          (:Is%20Valid
            (Class|CreativeMoveToolBlueprintPreview|ServerSetFXState (Class|FortCreativeMoveTool|FindActivePreviewforObject _returnvalue _array_element) "TransitionOut")
            (Class|CreativeMoveToolBlueprintPreview|ServerSetPlayTransitionBackwardswhenCurrentAnimEnds (Class|FortCreativeMoveTool|FindActivePreviewforObject _returnvalue _array_element))))))))

(event Custom|SpawnedObjectEffects
  (params (Components :type Array%20of%20Mesh%20Component%20Object%20References))
  ;[1 1 1] Server Create Spawn effects
  (CallFunction|SpawnBlueprintPreview (Components|GetOwner (Utilities|Array|Get(acopy) Components)) "TransitionOut"))

(event CreativeMoveTool|EventOnObjectPlacement
  (params (Components :type Array%20of%20Mesh%20Component%20Object%20References) (NewlyPlacedActorPair :type Original%20And%20Spawned%20Pair%20Structure))
  ;[1 1 1] On duplicate For the new actor. Currently Client only
  (CallFunction|SpawnedObjectEffects Components))

(event CreativeMoveTool|EventStartCreativeInteractionOnServer
  ;[1 1 1] Server Start Interacting
  (|CallParentFunction)
  (for _array_element (CreativeMoveTool|GetOwnerSelectedActors)
    (CallFunction|SpawnBlueprintPreview _array_element "TransitionIn")))

(event CreativeMoveTool|EventTriggerInteractionBehavior
  ;[1 1 1] Triggered
  (|CallParentFunction)
  (Audio|PlaySoundatLocation "/Game/Sounds/Creative/Gadgets/MoveTool/MoveTool_Place_01_Cue.MoveTool_Place_01_Cue" (Transformation|GetActorLocation (Components|GetOwner))))

(fn SpawnBlueprintPreview
  (params (Prop%20Actor :type Actor%20Object%20Reference%20%28by%20ref%29) (SpawnTransitionState :type ECreativeMoveToolPreviewFX%20Enum))
  ;[1 1 1] Make a Preview Actor On the Server
  (Utilities|IsValid Prop%20Actor
    (:Is%20Valid
      (CallFunction|SpawnPreview Prop%20Actor SpawnTransitionState))))
